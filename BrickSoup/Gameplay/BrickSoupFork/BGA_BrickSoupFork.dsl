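; BrickSoup gameplay actor that rewards players for dancing with an NPC bot by spawning holiday gift boxes.
; Its detection volume tracks players and grants or removes an external emote category.
; A repeating timer accumulates while an ungifted player dances, with the bot emoting along, then spawns one gift per eligible dancer and plays a gameplay cue.
; ScalableFloat rows let hotfixes disable gifting, emote granting or the whole actor.
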(event EventBeginPlay
  (bind _detectionvolume (Variables|Default|GetDetectionVolume))
  ; Enables destroying actor via Hotfix
  (if (ScalableFloat|GetValueasBool (Variables|Default|GetRow_IsEnabled))
    ; Enables removing "Emote!" widget via hotfix
    (if (ScalableFloat|GetValueasBool (Variables|Default|GetRow_ShouldGiveGift))
      (GameplayEventRouter|BroadcastEventBackwardCompatible (GameplayEventRouter|GetGlobalEventRouter) "(TagName=\"EventMessage.Athena.BrickSoup.ObjectiveActivated\")" (Utilities|Struct|MakeBrickSoupObjectiveActivated self))
      (else
        _))
    (Collision|BindEventtoOnComponentBeginOverlap _detectionvolume (delegate Custom|GrantEmoteEvent))
    (Collision|BindEventtoOnComponentEndOverlap _detectionvolume (delegate Custom|RemoveEmoteEvent))
    (if (ScalableFloat|GetValueasBool (Variables|Default|GetRow_ShouldGiveGift))
      (Utilities|Time|SetTimerbyEvent (EventDispatchers|CreateEvent :!event Custom|ProgressTimer) (Variables|Default|GetRefreshRate) true true))
    (else
      (Actor|DestroyActor))))

(event Custom|GrantEmoteEvent
  (params (OverlappedComponent :type Primitive%20Component%20Object%20Reference) (OtherActor :type Actor%20Object%20Reference) (OtherComp :type Primitive%20Component%20Object%20Reference) (OtherBodyIndex :type Integer) (bFromSweep :type Boolean) (SweepResult :type Hit%20Result%20Structure%20%28by%20ref%29))
  (bind _asfort_player_pawn (Utilities|Casting|CastToFortPlayerPawn OtherActor))
  (Utilities|Array|Add (Variables|Default|GetPlayerList) _asfort_player_pawn)
  (Development|PrintString "Added Player")
  (if (ScalableFloat|GetValueasBool (Variables|Default|GetRow_ShouldGrantEmote))
    (CallFunction|GrantEmote _asfort_player_pawn)
    (else
      _))
  (Default|CallOnPlayerAdded))

(event Custom|RemoveEmoteEvent
  (params (OverlappedComponent :type Primitive%20Component%20Object%20Reference) (OtherActor :type Actor%20Object%20Reference) (OtherComp :type Primitive%20Component%20Object%20Reference) (OtherBodyIndex :type Integer))
  (bind _asfort_player_pawn (Utilities|Casting|CastToFortPlayerPawn OtherActor))
  (Utilities|Array|RemoveItem (Variables|Default|GetPlayerList) _asfort_player_pawn)
  (if (ScalableFloat|GetValueasBool (Variables|Default|GetRow_ShouldGrantEmote))
    (CallFunction|RemoveEmote _asfort_player_pawn)
    (else
      _))
  (Default|CallOnPlayerRemoved))

(event Custom|ProgressTimer
  (bind _isvalid (CallFunction|CheckValidPlayers))
  (if _isvalid
    (if (not (Variables|Default|GetIsBotPlayingEmote))
      (CallFunction|StartBotEmote))
    (bind _output_get (Variables|Default|SetTimer (+ (Variables|Default|GetRefreshRate) (Variables|Default|GetTimer))))
    (Development|PrintString (Utilities|String|ToString(Float) _output_get))
    (if (>= _output_get (ScalableFloat|GetValueatLevel (Variables|Default|GetRow_TimerDuration)))
      (CallFunction|GrantPresent))
    (else
      (if (Variables|Default|GetIsBotPlayingEmote)
        (CallFunction|StopBotEmote)
        (else
          _))
      (Variables|Default|SetTimer 0.0))))

(fn CheckValidPlayers
  (if (> (Utilities|Array|Length (Variables|Default|GetPlayerList)) 0)
    (for _array_element (Variables|Default|GetPlayerList)
      (bind _botcontroller (Variables|Default|GetBotController)
        (:then
          (if (== (Variables|/Script/FortniteAI.FortAthenaAIBotController|GetPlayerBotPawn _botcontroller) _array_element)
            _
            (elif (Utilities|Array|ContainsItem (Variables|Default|GetGiftReceivedPlayerLIst) _array_element)
              (Development|PrintString "Player already received gift")
              (elif (GameplayTags|HasMatchingGameplayTag _array_element "(TagName=\"Gameplay.Action.Player.Emote.Dance\")")
                (LocalVariables|SetIsJamming true)))))
        (:else
          _)))
    (if (LocalVariables|GetIsJamming)
      (return true)
      (else
        (return false)))
    (else
      (return false))))

(fn GrantEmote
  (params (PlayerPawn :type Fort%20Player%20Pawn%20Object%20Reference))
  (bind _returnvalue (Fortnite|GetFortPlayerControllerfromActor PlayerPawn))
  (bind _returnvalue_2 (Actor|GetComponentbyClass _returnvalue "/Script/FortniteGame.FortControllerComponent_ExternalEmote"))
  (Utilities|IsValid _returnvalue_2
    (:Is%20Valid
      (Class|FortControllerComponentExternalEmote|AddExternalMultiEmoteCategories _returnvalue_2 (Utilities|Array|MakeArray (Variables|Default|GetEmoteCategory))))))

(fn GrantPresent
  (bind _giftreceivedplayerlist (Variables|Default|GetGiftReceivedPlayerLIst))
  (for _array_element (Variables|Default|GetPlayerList)
    (if (Pawn|IsBotControlled _array_element)
      _
      (elif (Utilities|Array|ContainsItem _giftreceivedplayerlist _array_element)
        _
        (elif (GameplayTags|HasMatchingGameplayTag _array_element "(TagName=\"Gameplay.Action.Player.Emote.Dance\")")
          (Utilities|Array|Add _giftreceivedplayerlist _array_element)
          (Inventory|SpawnPickupInWorld "/BrickSoup/Gameplay/BrickSoupFork/HolidayGift/Athena_BrickSoup_HolidayGiftBox.Athena_BrickSoup_HolidayGiftBox" 1 (+ (Transformation|GetWorldLocation (Variables|Default|GetDetectionVolume)) "0.000000,0.000000,150.000000"))))))
  (CallFunction|StopBotEmote)
  (Variables|Default|SetTimer 0.0)
  (Ability|ExecuteGameplayCueReplicated (Variables|Default|GetGC_Gift)))

(fn RemoveEmote
  (params (PlayerPawn :type Fort%20Player%20Pawn%20Object%20Reference))
  (bind _returnvalue (Fortnite|GetFortPlayerControllerfromActor PlayerPawn))
  (bind _returnvalue_2 (Actor|GetComponentbyClass _returnvalue "/Script/FortniteGame.FortControllerComponent_ExternalEmote"))
  (Utilities|IsValid _returnvalue_2
    (:Is%20Valid
      (Class|FortControllerComponentExternalEmote|RemoveExternalMultiEmoteCategories _returnvalue_2 (Utilities|Array|MakeArray (Utilities|Struct|BreakExternalMultiEmoteCategory (Variables|Default|GetEmoteCategory)))))))

(fn SetNPC
  (params (BotController :type Fort%20Athena%20AIBot%20Controller%20Object%20Reference))
  (bind _output_get (Variables|Default|SetBotController BotController))
  (CallFunction|GrantEmote (Variables|/Script/FortniteAI.FortAthenaAIBotController|GetPlayerBotPawn _output_get)))

(fn StartBotEmote
  (Class|FortAthenaAIBotController|RequestEmote (Variables|Default|GetBotController) (Variables|Default|GetBotEmote) 15.0)
  (Variables|Default|SetIsBotPlayingEmote true))

(fn StopBotEmote
  (Class|FortAthenaAIBotController|StopEmote (Variables|Default|GetBotController))
  (Variables|Default|SetIsBotPlayingEmote false))

(fn ConstructionScript)
