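; BP_DebugNPCSpawner_BrickSoupFork is, by its name, a debug NPC spawner Blueprint for the BrickSoupFork gameplay variant, and it adds no graph logic of its own.
; OnPawnSpawned passes the new pawn and request ID straight to the parent class, and the construction script also just calls its parent.
; BeginPlay and ActorBeginOverlap are disabled placeholder events, so behavior in these graphs is entirely inherited.
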(event EventBeginPlay
  (|CallParentFunction) ; This node is disabled and will not be called.\nDrag off pins to build functionality.
)

(event Collision|EventActorBeginOverlap
  (params (OtherActor :type Actor%20Object%20Reference))
  (|CallParentFunction OtherActor) ; This node is disabled and will not be called.\nDrag off pins to build functionality.
)

(fn OnPawnSpawned
  (params (NewPawn :type Pawn%20Object%20Reference) (RequestId :type Integer))
  (|CallParentFunction NewPawn RequestId))

(fn ConstructionScript
  (|CallParentFunction))
